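# Glowstick prop material, flagged for particle sprite use, that cycles its glow through disco gradient colors over time.
# Time scaled by LightSpeed is wrapped with Frac as the gradient lookup, giving one uniform changing color.
# That color is masked by the diffuse texture's red channel, scaled by Glow and passed through MF_FX_EmissiveAdjustments.
# The graph also holds a particle-color tint of the masked color and a mask texture sample.
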Material /BRCosmetics/Animation/Props/Glowstick/Materials/M_GlowStick_Dance
graph {
 properties {bUsedWithParticleSprites=true, pixelDepthOffsetMode=PDOM_Legacy, usedShadingModels=Default Lit}
 nodes [
  n1 = Multiply(A=Time(), B=ScalarParameter() {defaultValue=1, parameterName=LightSpeed})
  n2 = Multiply(
   A=TextureSample(UVs=Frac(None=@n1)) {texture=/Game/Athena/Events/NewYear/DiscoNY_Gradient}.RGB
   B=TextureSample() {texture=/BRCosmetics/Animation/Props/Glowstick/Textures/T_Glow_Stick_D}.R
  )
  MaterialFunctionCall(Emissive=Multiply(A=@n2, B=ScalarParameter() {defaultValue=16.358, parameterName=Glow})) {materialFunction=&MF_FX_EmissiveAdjustments}
  Multiply(A=ParticleColor().RGB, B=@n2)
  TextureSample() {samplerType=SAMPLERTYPE_Masks, texture=/BRCosmetics/Animation/Props/Glowstick/Textures/T_Glow_Stick_S}
 ]
}
