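# Glowstick prop material for Papaya with a color-driven glow that pulses to event audio.
# The MF_Papaya_GadgetColor function supplies the color, which is masked by the diffuse texture's red channel.
# The EventAudio_Amplitude collection parameter blends between GlowLow and GlowHigh to scale a brighter branch that applies the red mask twice, tightening the glow.
# A separate mask texture is also sampled in this Default Lit material.
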Material /BRCosmetics/Animation/Props/Glowstick/Materials/M_GlowStick_Papaya
graph {
 properties {pixelDepthOffsetMode=PDOM_Legacy, usedShadingModels=Default Lit}
 nodes [
  n1 = MaterialFunctionCall!(
   Media Player TexCoord
   Media Color Code Coordinate
   Media Texture Resolution
   Base Color Parameter
   Base Color Mix
   Media Color Code Mix
   Intermission Color Parameter
   Intermission Color Mix
   Desaturation
  ) {materialFunction=/Game/Environments/Papaya/Materials/MaterialFunctions/MF_Papaya_GadgetColor}
  n2 = CollectionParameter() {collection=&FortniteMaterialParameters, parameterName=EventAudio_Amplitude}
  n3 = LinearInterpolate(A=ScalarParameter() {defaultValue=2, parameterName=GlowLow}, B=ScalarParameter() {defaultValue=40, parameterName=GlowHigh}, Alpha=@n2)
  n4 = TextureSample() {texture=/BRCosmetics/Animation/Props/Glowstick/Textures/T_Glow_Stick_D}
  Multiply(A=Multiply(A=Multiply(A=@n1.Result, B=@n3), B=@n4.R), B=@n4.R)
  Multiply(A=@n1.Result, B=@n4.R)
  TextureSample() {samplerType=SAMPLERTYPE_Masks, texture=/BRCosmetics/Animation/Props/Glowstick/Textures/T_Glow_Stick_S}
 ]
}
