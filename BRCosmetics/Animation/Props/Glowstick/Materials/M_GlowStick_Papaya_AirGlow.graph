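# Additive unlit glow for the air around the Papaya glowstick prop, using a camera-facing mesh pushed toward the camera by its radius.
# A radial halo combines with fake lighting from scene normals, diffuse color and reconstructed position behind the translucency.
# Gadget color and event audio amplitude tint and pulse the result, which fades by distance and depth.
# Forward and mobile paths collapse the mesh to its pivot and output black.
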Material /BRCosmetics/Animation/Props/Glowstick/Materials/M_GlowStick_Papaya_AirGlow
graph {
 properties {bEnableResponsiveAA=true, blendMode=BLEND_Additive, pixelDepthOffsetMode=PDOM_Legacy, shadingModel=MSM_Unlit, usedShadingModels=Unlit}
 nodes [
  ScalarParameter() {defaultValue=10, parameterName=CameraOffset}
  n1 = WorldPosition()
  n2 = Transform(None=Constant3Vector() {constant=<rgba 0 0 1 0>}) {transformSourceType=TRANSFORMSOURCE_Local}
  n3 = Transform(None=Constant3Vector() {constant=<rgba 0 1 0 0>}) {transformSourceType=TRANSFORMSOURCE_Local}
  n4 = Transform(None=Constant3Vector() {constant=<rgba 1 0 0 0>}) {transformSourceType=TRANSFORMSOURCE_Local}
  n5 = Normalize(VectorInput=Subtract(A=CameraPositionWS(), B=ObjectPositionWS()))
  n6 = ObjectRadius()
  n7 = Subtract(A=ObjectPositionWS(), B=WorldPosition().XYZ)
  ShadingPathSwitch(
   Default=Add(
    A=MaterialFunctionCall!(WorldPosition=@n1.XYZ, PivotLocation, WS Normals, Custom Object Basis 1=@n2, Custom Object Basis 2=@n3, Custom Object Basis 3=@n4) {materialFunction=/Engine/Functions/Engine_MaterialFunctions02/WorldPositionOffset/AlignMeshToTheCamera}.World Position Offset
    B=Multiply(A=@n5, B=@n6)
   )
   Forward=@n7
   Mobile=@n7
  )
  n8 = ScalarParameter() {defaultValue=40, parameterName=FadeStartDistance}
  n9 = ObjectRadius()
  n10 = ScalarParameter() {defaultValue=10, parameterName=FadeEndDistance}
  n11 = ObjectRadius()
  n12 = VertexInterpolator(VS=MaterialFunctionCall(Position=ObjectPositionWS(), Fade_Start_Distance=Add(A=@n8, B=@n9), Fade_End_Distance=Add(A=@n10, B=@n11)) {materialFunction=&MF_Distance_SyncedFade_FX}.Result)
  n13 = Constant() {r=0.5}
  n14 = SphereMask(A=TextureCoordinate(), B=@n13, Radius=@n13, Hardness=Constant())
  n15 = Multiply(A=Multiply(A=@n14, B=@n14), B=ScalarParameter() {defaultValue=0.055, parameterName=GlowOpacity, primitiveDataIndex=5})
  n16 = MaterialFunctionCall() {materialFunction=&ObjectPivotPoint}
  n17 = MaterialFunctionCall() {materialFunction=&WorldPositionBehindTranslucency}
  n18 = Normalize(VectorInput=Subtract(A=@n16.Object Pivot Location, B=@n17.Result))
  n19 = ComponentMask(None=SceneTexture() {sceneTextureId=PPI_WorldNormal}.Color) {b=true, g=true, r=true}
  n20 = ObjectRadius()
  n21 = Multiply(
   A=Saturate(None=DotProduct(A=@n18, B=@n19))
   B=SphereMask(
    A=@n16.Object Pivot Location
    B=@n17.Result
    Radius=Multiply(A=@n20) {constB=0.725}
    Hardness=ScalarParameter() {defaultValue=0.1, parameterName=FakeLightingMaskHardness}
   )
  )
  n22 = ComponentMask(None=SceneTexture() {sceneTextureId=PPI_DiffuseColor}.Color) {b=true, g=true, r=true}
  n23 = ScalarParameter() {defaultValue=0.7, parameterName=GlowLow}
  n24 = ScalarParameter() {defaultValue=3.5, parameterName=GlowHigh}
  n25 = CollectionParameter() {collection=&FortniteMaterialParameters, parameterName=EventAudio_Amplitude}
  n26 = Multiply(
   A=Add(A=Multiply(A=@n12.PS, B=@n15), B=Multiply(A=@n21, B=@n22))
   B=Multiply(
    A=MaterialFunctionCall!(
     Media Player TexCoord
     Media Color Code Coordinate
     Media Texture Resolution
     Base Color Parameter
     Base Color Mix
     Media Color Code Mix
     Intermission Color Parameter
     Intermission Color Mix
     Desaturation
    ) {materialFunction=/Game/Environments/Papaya/Materials/MaterialFunctions/MF_Papaya_GadgetColor}.Result
    B=LinearInterpolate(A=@n23, B=@n24, Alpha=@n25)
   )
  )
  n27 = DepthFade(FadeDistance=ScalarParameter() {defaultValue=20, parameterName=FadeDistance})
  n28 = Constant()
  ShadingPathSwitch(Default=Multiply(A=ComponentMask(None=@n12.PS) {r=true}, B=Multiply(A=@n26, B=@n27)), Forward=@n28, Mobile=@n28)
 ]
}
